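# Niagara system for the GnocchiTea emote, layering glow, rune, ring, smoke and debris effects around the puzzle gadget prop over about five seconds.
# Light flares and dust bursts mark the start and end, with gem glows and smoke circles delayed until the end.
# Socket and skeletal mesh sampling pins glows, runes and rings to their locations.
# Time-of-day inputs tint dust and smoke, while a front-end preview flag dims some glows.
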NiagaraSystem /BRCosmetics/Animation/Game/MainPlayer/Emotes/GnocchiTea/FX/NS_Emote_GnocchiTea_Burst
dependencies {
 usedDataInterfaces [
  &NiagaraDataInterfaceCurve
  &NiagaraDataInterfaceColorCurve
  &NiagaraDataInterfaceVector2DCurve
  &NiagaraDataInterfaceSpriteRendererInfo
  &NiagaraDataInterfaceArrayColor
  &NiagaraDataInterfaceSocketReader
  &NiagaraDataInterfaceMeshRendererInfo
  &NiagaraDataInterfaceSkeletalMesh
 ]
 usedDynamicInputs [&Vector2DFromCurve]
 usedModules [
  &self:SystemSpawnScript.SystemScriptSource.SystemScriptGraph.NiagaraNodeAssignment_0.SetVariables_#1
  &self:SystemSpawnScript.SystemScriptSource.SystemScriptGraph.NiagaraNodeAssignment_2.SetVariables_#2
  &self:SystemSpawnScript.SystemScriptSource.SystemScriptGraph.NiagaraNodeAssignment_1.SetVariables_#3
  &self:SystemSpawnScript.SystemScriptSource.SystemScriptGraph.NiagaraNodeAssignment_3.SetVariables_#4
  &SystemState
 ]
}
emitters {
 Chunks001 {
  simTarget=GPUComputeSim
  emitterUpdateScript [
   &EmitterState(Life Cycle Mode=::NewEnumerator1, Loop Behavior=::NewEnumerator1, Loop Delay=0.05, Loop Duration=1, UseLoopDelay=false)
   &SpawnBurst_Instantaneous(
    Spawn Count=&RandomRangeInt(Minimum=10, Maximum=7)
    Use Spawn Probability=false
    UseLoopCountLimit=false
   )
  ]
  particleSpawnScript [
   &InitializeParticle(
    Color Maximum=<rgba 0.2083 0.139 0.0696 1>
    Color Minimum=<rgba 0.333 0.2498 0.1665 1>
    Color Mode=::NewEnumerator2
    Lifetime Max=0.2
    Lifetime Min=0.5
    Lifetime Mode=::NewEnumerator1
    Mass Mode=::NewEnumerator1
    Sprite Rotation Mode=::NewEnumerator2
    Sprite Size Mode=::NewEnumerator1
    Sprite UV Mode=::NewEnumerator3
    Uniform Sprite Size Max=8
    Uniform Sprite Size Min=4
    UsePositionOffset=false
    Write Material Random=true
   )
   &AddVelocity(Velocity=&RandomRangeVector(Minimum=<-200 0 -100>, Maximum=<200 200 200>), Velocity Speed Scale=2)
  ]
  particleUpdateScript [
   &ParticleState()
   &GravityForce()
   &Drag(Drag=2, Use Linear Drag=true, Use Rotational Drag=true)
   &SolveForcesAndVelocity(Clamp Velocity=false, Limit Acceleration=false)
   &ScaleSpriteSize(Scale Factor=&Vector2DFromFloat(Value=&FloatFromCurve(FloatCurve=DataInterface() {propertyValues {Curve {keys [<at 0.5 1 RCIM_Cubic>, {arriveTangent=-3.7562, interpMode=RCIM_Cubic, leaveTangent=-3.7562, tangentMode=RCTM_User, time=1, value=0}]}}})))
  ]
  rendererClasses [&NiagaraSpriteRendererProperties]
  renderers [{rendererClass=&NiagaraSpriteRendererProperties}]
 }
 DustBursts {
  emitterUpdateScript [
   &EmitterState(Life Cycle Mode=::NewEnumerator1, Loop Behavior=::NewEnumerator1, Loop Delay=1.9, Loop Duration=6, UseLoopDelay=false)
   &SpawnBurst_Instantaneous(Spawn Count=7, Use Spawn Probability=false, UseLoopCountLimit=false)
   &SpawnBurst_Instantaneous(Spawn Count=10, Spawn Group=1, Spawn Time=5.05, Use Spawn Probability=false, UseLoopCountLimit=false) {module=SpawnBurst_Instantaneous001}
  ]
  particleSpawnScript [
   &InitializeParticle(
    Color Minimum=<rgba 0.2587 0.2587 0.2587 1>
    Color Mode=::NewEnumerator2
    Lifetime Max=0.4
    Lifetime Min=0.6
    Lifetime Mode=::NewEnumerator1
    Sprite Rotation Mode=::NewEnumerator2
    Sprite Size Mode=::NewEnumerator3
    Uniform Sprite Size=&MaskFloatBySpawnGroup(
     Masked Float=&RandomRangeFloat(Minimum=150, Maximum=200)
     Passthrough Float=&RandomRangeFloat(Minimum=40, Maximum=60)
    )
    UsePositionOffset=false
    Write Material Random=true
   )
   &ShapeLocation(
    MaskSpawnGroup=false
    Override Seed=false
    Sphere Radius=&MaskFloatBySpawnGroup(Masked Float=30, Passthrough Float=15)
    Sphere Surface Distribution=1
    UseOwnerScale=true
   )
   &CameraOffset(Camera Offset Amount=&RandomRangeFloat(Minimum=0, Maximum=10))
   &AddVelocity(
    Module.Use Falloff Distance=false
    Override Seed=false
    Velocity Mode=::NewEnumerator1
    Velocity Speed=&MaskFloatBySpawnGroup(
     Masked Float=&RandomRangeFloat(Minimum=100, Maximum=50)
     Passthrough Float=&RandomRangeFloat(Minimum=60, Maximum=30)
    )
   )
  ]
  particleUpdateScript [
   &ParticleState()
   &SpriteRotationRate(Rotation Rate=&RandomRangeFloat(Minimum=-150, Maximum=150))
   &ScaleSpriteSize(Uniform Curve Sprite Scale=DataInterface() {propertyValues {Curve {keys [
        {arriveTangent=1.6237, interpMode=RCIM_Cubic, leaveTangent=1.6237, tangentMode=RCTM_User, time=0, value=1}
        {arriveTangent=0.637, tangentMode=RCTM_User, time=1, value=1.7}
       ]}}})
   &Drag(
    Drag=&FloatFromCurve(FloatCurve=DataInterface() {propertyValues {Curve {keys [{interpMode=RCIM_Cubic, tangentMode=RCTM_User, time=0, value=0}, {interpMode=RCIM_Cubic, tangentMode=RCTM_User, time=0.4943, value=1}]}}})
    Use Linear Drag=true
    Use Rotational Drag=true
   )
   &SolveForcesAndVelocity(Clamp Velocity=false, Limit Acceleration=false)
   &ScaleColor(
    Scale Alpha=&FloatFromCurve(
     FloatCurve=DataInterface() {propertyValues {Curve {keys [
         {interpMode=RCIM_Cubic, tangentMode=RCTM_User, time=0, value=0.8902}
         {arriveTangent=-2.0189, interpMode=RCIM_Cubic, leaveTangent=-2.1015, tangentMode=RCTM_Break, time=0.2, value=1.8153}
         {arriveTangent=-1.1521, interpMode=RCIM_Cubic, tangentMode=RCTM_Break, time=1, value=0}
        ]}}}
     Scale Curve=&FloatFromCurve(
      FloatCurve=DataInterface() {propertyValues {Curve {keys [{interpMode=RCIM_Cubic, tangentMode=RCTM_User, time=4.5123, value=1}, {interpMode=RCIM_Cubic, tangentMode=RCTM_User, time=4.95, value=0}]}}}
      Scale Curve=&MaskFloatBySpawnGroup(
       Masked Float=1
       Passthrough Float=&FNTimeOfDayScalar(Night Scale=0.6, Morning Scale=0.4, Day Scale=0.3, Evening Scale=0.4)
      )
     )
    )
    Scale RGB=&MakeVectorFromLinearColorRGB(Color=&FNTimeOfDayColor(
      Night Color=&ColorFromCurve(ColorCurve=DataInterface() {propertyValues {
         AlphaCurve {keys [<at 0 1 RCIM_Cubic>]}
         BlueCurve {keys [<at 0 0.8096 RCIM_Cubic>,<at 1 0.3698 RCIM_Cubic>]}
         GreenCurve {keys [<at 0 0.4742 RCIM_Cubic>,<at 1 0.2166 RCIM_Cubic>]}
         RedCurve {keys [<at 0 0.4483 RCIM_Cubic>,<at 1 0.2048 RCIM_Cubic>]}
        }})
      Morning Color=&ColorFromCurve(ColorCurve=DataInterface() {propertyValues {
         AlphaCurve {keys [<at 0 1 RCIM_Cubic>]}
         BlueCurve {keys [<at 0 0.256 RCIM_Cubic>,<at 1 0.145 RCIM_Cubic>]}
         GreenCurve {keys [<at 0 0.3327 RCIM_Cubic>,<at 1 0.1631 RCIM_Cubic>]}
         RedCurve {keys [<at 0 0.8096 RCIM_Cubic>,<at 1 0.4161 RCIM_Cubic>]}
        }})
      Day Color=&ColorFromCurve(ColorCurve=DataInterface() {propertyValues {
         AlphaCurve {keys [<at 0 1 RCIM_Cubic>]}
         BlueCurve {keys [<at 0 0.0659 RCIM_Cubic>,<at 1 0.0754 RCIM_Cubic>]}
         GreenCurve {keys [<at 0 0.4818 RCIM_Cubic>,<at 1 0.3164 RCIM_Cubic>]}
         RedCurve {keys [<at 0 0.8096 RCIM_Cubic>,<at 1 0.8096 RCIM_Cubic>]}
        }})
      Evening Color=&ColorFromCurve(ColorCurve=DataInterface() {propertyValues {
         AlphaCurve {keys [<at 0 1 RCIM_Cubic>]}
         BlueCurve {keys [<at 0 0.3766 RCIM_Cubic>,<at 1 0.2029 RCIM_Cubic>]}
         GreenCurve {keys [<at 0 0.4646 RCIM_Cubic>,<at 1 0.2332 RCIM_Cubic>]}
         RedCurve {keys [<at 0 0.4693 RCIM_Cubic>,<at 1 0.2195 RCIM_Cubic>]}
        }})
     ))
    ScaleA=true
    ScaleRGB=true
   )
  ]
  rendererClasses [&NiagaraSpriteRendererProperties]
  renderers [{rendererClass=&NiagaraSpriteRendererProperties}]
 }
 GemGlows001 {
  emitterUpdateScript [
   &EmitterState(Life Cycle Mode=::NewEnumerator1, Loop Behavior=::NewEnumerator1, Loop Delay=5.13, Loop Duration=1, UseLoopDelay=true)
   &SpawnBurst_Instantaneous(Spawn Count=3, Spawn Group=1, Use Spawn Probability=false, UseLoopCountLimit=false) {module=SpawnBurst_Instantaneous001}
   &SpawnBurst_Instantaneous(Spawn Count=3, Use Spawn Probability=false, UseLoopCountLimit=false)
  ]
  particleSpawnScript [
   &PartitionParticles(Number Of Partitions=3)
   &FN_SetSpawnGroup(Number of Spawn Groups=2)
   &InitializeParticle(
    Color=&MaskLinearColorBySpawnGroup(Masked Linear Color=<rgba 0.1181 1 0.3548 1>, Passthrough Linear Color=<rgba 1 0.3847 0 0.4493>)
    Sprite Size Mode=::NewEnumerator3
    Uniform Sprite Size=17
    UsePositionOffset=false
    Write Material Random=true
   )
   &CameraOffset(Camera Offset Amount=10)
  ]
  particleUpdateScript [
   &ParticleState()
   &SkeletalMeshLocation(
    Bone / Socket Index=$Particles.PartitionParticles.Partition
    Bone Sampling Mode=::NewEnumerator4
    MaskSpawnGroup=false
    Sampled Position Offset=&MaskVectorBySpawnGroup(Masked Vector=<0 -6 0>, Passthrough Vector=<0 0 0>)
    Skeletal Mesh=$System.GemLocation
   )
   &ScaleColor(
    Scale Alpha=&FloatFromCurve(
     FloatCurve=DataInterface() {propertyValues {Curve {keys [
         {arriveTangentWeight=0.0665, interpMode=RCIM_Cubic, leaveTangentWeight=0.0103, tangentMode=RCTM_User, time=0.0908, value=6}
         {
          arriveTangent=-0.1272
          arriveTangentWeight=1.3428
          interpMode=RCIM_Cubic
          leaveTangent=-0.1272
          tangentMode=RCTM_User
          time=0.3952
          value=1.3473
         }
         <at 1 0 RCIM_Cubic>
        ]}}}
     CurveIndex=$Particles.Age
     Scale Curve=&MaskFloatBySpawnGroup(
      Masked Float=&FNTimeOfDayScalar(Night Scale=0.4, Morning Scale=0.4, Day Scale=0.4, Evening Scale=0.4)
      Passthrough Float=&FNTimeOfDayScalar(Night Scale=0.7, Morning Scale=1, Evening Scale=1)
     )
    )
    ScaleA=true
    ScaleRGB=true
   )
  ]
  rendererClasses [&NiagaraSpriteRendererProperties]
  renderers [{rendererClass=&NiagaraSpriteRendererProperties}]
 }
 Light {
  emitterUpdateScript [
   &EmitterState(Life Cycle Mode=::NewEnumerator1, Loop Behavior=::NewEnumerator1, Loop Duration=5.1, UseLoopDelay=true)
   &SpawnBurst_Instantaneous(Use Spawn Probability=false, UseLoopCountLimit=false)
  ]
  particleSpawnScript [
   &InitializeParticle(
    Color=&MakeCustomLinearColorFromBool(Bool=$User.bIsFrontEndPreview, True Color=<rgba 0.3185 0.0137 0 1.3942>, False Color=<rgba 0.3177 0.0139 0 3>)
    Lifetime=5.25
    Position Offset=<0 0 -10>
    Sprite Size Mode=::NewEnumerator3
    Uniform Sprite Size=250
    UsePositionOffset=false
    Write Material Random=true
   )
   /SetVariables(
    Particles.CenterCameraOffset=20
    Particles.CenterSize:Vector2f=<88.5783 88.5783>
    Particles.Vector 4:Vector4f=&MakeVector4(X=&RandomRangeFloat(Minimum=0.7, Maximum=1), Y=1, Z=1, W=1)
   ) {isSetParameters=true}
  ]
  particleUpdateScript [
   &ParticleState()
   &SpriteRotationRate(Rotation Rate=&FloatFromCurve(
     FloatCurve=DataInterface() {propertyValues {Curve {keys [
         {interpMode=RCIM_Cubic, tangentMode=RCTM_User, time=0, value=1}
         {arriveTangent=-1.2549, interpMode=RCIM_Cubic, leaveTangent=-1.2549, tangentMode=RCTM_User, time=1, value=0}
        ]}}}
     Scale Curve=20
    ))
   &ScaleColor(
    Linear Color Curve=DataInterface() {propertyValues {
      AlphaCurve {keys [<at 0 0>,<at 0.0244 1>,<at 0.0395 0.5829>,<at 0.9564 0.1804>,<at 0.97 0.6>,<at 1 0>]}
      BlueCurve {keys [<at 0 1>]}
      GreenCurve {keys [<at 0 1>]}
      RedCurve {keys [<at 0 1>]}
     }}
    Scale Mode=::NewEnumerator2
   )
   &CameraOffset(Camera Offset Amount=-10)
   /SetVariables(
    Particles.FlareColor=&ColorFromCurve(
     ColorCurve=DataInterface() {propertyValues {
       AlphaCurve {keys [
         <at 0 0>
         {arriveTangent=11.236, interpMode=RCIM_Cubic, leaveTangent=11.236, time=0.022, value=1}
         <at 0.089 1>
         <at 0.157 0.25>
         {arriveTangentWeight=0.0649, interpMode=RCIM_Cubic, leaveTangentWeight=0.1733, tangentMode=RCTM_User, time=0.32, value=0}
         <at 5.01 0.0074>
         <at 5.05 0.4435>
         <at 5.1 1>
         <at 5.2 0>
        ]}
       BlueCurve {keys [<at 0 0 RCIM_Cubic>,<at 0.45 0 RCIM_Cubic>,<at 0.7964 0>,<at 0.9948 0>]}
       GreenCurve {keys [
         <at 0 4.2659 RCIM_Cubic>
         {arriveTangentWeight=0.1997, interpMode=RCIM_Cubic, leaveTangentWeight=1.9287, tangentMode=RCTM_User, time=0.45, value=0.051}
         <at 4.9 0.051>
         <at 5.1 5.3387>
        ]}
       RedCurve {keys [
         <at 0 10 RCIM_Cubic>
         {arriveTangentWeight=0.3189, interpMode=RCIM_Cubic, leaveTangentWeight=3.0805, tangentMode=RCTM_User, time=0.45, value=1}
         <at 4.9 1>
         <at 5.1 10>
        ]}
      }}
     CurveIndex=$Particles.Age
    )
    Particles.FlareDynPar:Vector4f=<1 0 0 0>
    Particles.FlareSpriteSize:Vector2f=&Vector2DFromFloat(Value=&FloatFromCurve(
      FloatCurve=DataInterface() {propertyValues {Curve {keys [
          {interpMode=RCIM_Cubic, leaveTangent=5.6981, tangentMode=RCTM_Break, time=0, value=0.2}
          <at 0.315 1>
          {arriveTangent=-0.209, interpMode=RCIM_Cubic, leaveTangent=-0.209, time=4.9, value=1}
          {arriveTangent=-13.1443, interpMode=RCIM_Cubic, leaveTangent=-13.1443, tangentMode=RCTM_User, time=5.1, value=0}
         ]}}}
      CurveIndex=$Particles.Age
      Scale Curve=320
     ))
   ) {isSetParameters=true}
   /SetVariables(
    Particles.RaysColor:LinearColor=&ColorFromCurve(ColorCurve=DataInterface() {propertyValues {
       AlphaCurve {keys [
         <at 0 0>
         {arriveTangent=22.9527, interpMode=RCIM_Cubic, leaveTangent=22.9527, time=0.0127, value=0.7091}
         <at 0.0309 0.7091>
         {arriveTangent=-0.0704, interpMode=RCIM_Cubic, leaveTangent=-0.0704, tangentMode=RCTM_User, time=0.938, value=0.2091}
         <at 0.952 0.6091>
         {arriveTangent=-0.5244, interpMode=RCIM_Cubic, leaveTangent=-0.5244, tangentMode=RCTM_User, time=0.97, value=0}
        ]}
       BlueCurve {keys [
         <at 0 0 RCIM_Cubic>
         {arriveTangent=0.1387, interpMode=RCIM_Cubic, leaveTangent=0.1387, time=0.025, value=0.1293}
         <at 0.935 0.1296>
         <at 0.95 0>
         <at 0.97 0.1293>
        ]}
       GreenCurve {keys [
         <at 0 2.3556 RCIM_Cubic>
         {arriveTangent=-1.5465, interpMode=RCIM_Cubic, leaveTangent=-1.5465, time=0.025, value=0.9131}
         <at 0.935 0.9096>
         <at 0.95 3.294>
         <at 0.97 0.9131>
        ]}
       RedCurve {keys [
         <at 0 6 RCIM_Cubic>
         {arriveTangent=-4.8244, interpMode=RCIM_Cubic, leaveTangent=-4.8244, time=0.025, value=1.5}
         <at 0.935 1.4892>
         <at 0.95 6>
         <at 0.97 1.5>
        ]}
      }})
    Particles.RaysDynPar:Vector4f=&MakeVector4(
     X=&FloatFromCurve(FloatCurve=DataInterface() {propertyValues {Curve {keys [<at 0 0>,<at 1 1>]}}})
     Y=0
     Z=0
     W=0
    )
    Particles.RaysSpriteSize:Vector2f=&Vector2DFromFloat(Value=&FloatFromCurve(
      FloatCurve=DataInterface() {propertyValues {Curve {keys [
          {interpMode=RCIM_Cubic, leaveTangentWeight=0.1397, tangentMode=RCTM_User, time=0.7837, value=0.7714}
          {arriveTangent=2.6513, arriveTangentWeight=0.1641, interpMode=RCIM_Cubic, leaveTangent=-0.002, tangentMode=RCTM_Break, time=0.98, value=1}
          {arriveTangentWeight=0.1989, interpMode=RCIM_Cubic, leaveTangentWeight=0.1346, tangentMode=RCTM_User, time=0.99, value=1}
          <at 0.995 0 RCIM_Cubic>
         ]}}}
      Scale Curve=350
     ))
   ) {isSetParameters=true}
   &DynamicMaterialParameters(
    Index 0 Param 1=1
    Index 0 Param 2=1
    Index 0 Param 3=1
    Index 0 Param 4=2.3
    Index 2 Param 1=2.5
    Index 2 Param 3=1
    Index 2 Param 4=1
    Index 3 Param 1=30
    Module.Index0Float0Write=true
    Module.Index0Float1Write=true
    Module.Index0Float2Write=true
    Module.Index0Float3Write=true
    Module.Index1Float0Write=true
    Module.Index1Float1Write=true
    Module.Index1Float2Write=true
    Module.Index1Float3Write=true
    Module.Index2Float0Write=true
    Module.Index2Float1Write=true
    Module.Index2Float2Write=true
    Module.Index2Float3Write=true
    Module.Index3Float0Write=true
    Module.Index3Float1Write=true
    Module.Index3Float2Write=true
    Module.Index3Float3Write=true
    Write Parameter Index 1=::NewEnumerator1
    Write Parameter Index 2=::NewEnumerator1
    Write Parameter Index 3=::NewEnumerator1
   )
   &FN_Light_Attributes(
    Color=<rgba 1 0.4415 0 1>
    Create Color Attribute=true
    Create Exponent Attribute=true
    Create Radius Attribute=true
    Enable Particle Light Enable=true
    Exponent=3
    Light Enabled=true
    Radius=300
    Scale Color=&FloatFromCurve(
     FloatCurve=DataInterface() {propertyValues {Curve {keys [
         {interpMode=RCIM_Cubic, tangentMode=RCTM_User, time=-0.0453, value=1}
         {arriveTangentWeight=0.7525, interpMode=RCIM_Cubic, leaveTangentWeight=0.7525, tangentMode=RCTM_User, time=0.0274, value=9.4525}
         {
          arriveTangent=-0.6171
          arriveTangentWeight=0.0629
          interpMode=RCIM_Cubic
          leaveTangent=-0.6171
          leaveTangentWeight=0.3771
          tangentMode=RCTM_User
          time=0.0683
          value=1
         }
         {arriveTangent=0.845, interpMode=RCIM_Cubic, leaveTangent=0.845, tangentMode=RCTM_User, time=0.95, value=1}
         {arriveTangentWeight=1.227, interpMode=RCIM_Cubic, leaveTangentWeight=0.9596, tangentMode=RCTM_User, time=0.98, value=4.2756}
         {interpMode=RCIM_Cubic, tangentMode=RCTM_User, time=1, value=0}
        ]}}}
     Scale Curve=4
    )
   )
   /SetVariables(
    Particles.CenterColor:LinearColor=&MakeLinearColor(
     R:NiagaraFloat=1
     G:NiagaraFloat=0.332
     B:NiagaraFloat=0
     A:NiagaraFloat=&Multiply_Float(
      A=&FNTimeOfDayScalar(Night Scale=0.5, Morning Scale=0.3, Evening Scale=0.4)
      B=&FloatFromCurve(FloatCurve=DataInterface() {propertyValues {Curve {keys [<at 0 0>,<at 0.0339 1>,<at 0.995 1>,<at 1 0>]}}})
     )
    )
    Particles.CenterSize:Vector2f=&Vector2DFromFloat(Value=&FloatFromCurve(FloatCurve=DataInterface() {propertyValues {Curve {keys [<at 0.9718 1.5 RCIM_Cubic>,<at 1 0>]}}}, Scale Curve=88.5783))
   ) {isSetParameters=true}
  ]
  rendererClasses [&NiagaraSpriteRendererProperties, &NiagaraLightRendererProperties]
  renderers [
   {rendererClass=&NiagaraSpriteRendererProperties}
   {rendererClass=&NiagaraSpriteRendererProperties, rendererIndex=1}
   {rendererClass=&NiagaraSpriteRendererProperties, rendererIndex=2}
   {rendererClass=&NiagaraLightRendererProperties, rendererIndex=3}
   {rendererClass=&NiagaraSpriteRendererProperties, rendererIndex=4}
  ]
 }
 PieceGlows {
  emitterUpdateScript [
   &EmitterState(Life Cycle Mode=::NewEnumerator1, Loop Behavior=::NewEnumerator1, UseLoopDelay=false)
   &SpawnBurst_Instantaneous(Spawn Count=6, Use Spawn Probability=false, UseLoopCountLimit=false)
  ]
  particleSpawnScript [
   &FN_SetSpawnGroup(Number of Spawn Groups=6)
   &InitializeParticle(
    Color=&MakeCustomLinearColorFromBool(Bool=$User.bIsFrontEndPreview, True Color=<rgba 1 0.7032 0.2957 0.3024>, False Color=<rgba 1 0.7032 0.2957 0.5729>)
    Lifetime=5
    Sprite Rotation Angle=&MaskFloatBySpawnGroup(Masked Float=0.406, Passthrough Float=0, Spawn Group Mask=1)
    Sprite Rotation Mode=::NewEnumerator1
    Sprite Size=&MaskVector2dBySpawnGroup(
     Masked Vector2D=<60 110>
     Passthrough Vector2D=&MaskVector2dBySpawnGroup(
      Masked Vector2D=<120 60>
      Passthrough Vector2D=&MaskVector2dBySpawnGroup(
       Masked Vector2D=<60 130>
       Passthrough Vector2D=&MaskVector2dBySpawnGroup(
        Masked Vector2D=<60 60>
        Passthrough Vector2D=&MaskVector2dBySpawnGroup(Masked Vector2D=<60 60>, Passthrough Vector2D=<60 60>, Spawn Group Mask=4)
        Spawn Group Mask=3
       )
       Spawn Group Mask=2
      )
      Spawn Group Mask=1
     )
     Spawn Group Mask=0
    )
    Sprite Size Mode=::NewEnumerator4
    UsePositionOffset=false
    Write Material Random=true
   )
   &CameraOffset(Camera Offset Amount=10)
  ]
  particleUpdateScript [
   &SocketLocation(
    SamplePosOffset=false
    Socket Index=$Particles.SpawnGroup
    Socket Reader=DataInterface() {propertyValues {
      AttachComponentClass=null
      AttachComponentTag=None
      EditorPreviewAsset=/BRCosmetics/Animation/Props/Gadgets_GnocchiTea/Meshes/GnocchiTea_Gadget
      SourceActorRef=null
      SourceAsset=&GnocchiTea_Gadget
      SourceMode=Default
      bRequireCurrentFrameData=true
      bUpdateSocketsPerFrame=true
      FilteredSockets [JNT_OuterPuzzle4, JNT_OuterPuzzle3, JNT_OuterPuzzle1, JNT_InnerPuzzle1, JNT_InnerPuzzle2, JNT_InnerPuzzle3]
      ObjectParameterBinding {parameter {name=None, typeDefHandle {registeredTypeIndex=1}, varData []}}
     }}
    Socket Sampling Mode=::NewEnumerator3
    VelocitySpeedLimit=false
   )
   &ParticleState()
   &ScaleColor(
    Scale Alpha=&FloatFromCurve(
     FloatCurve=DataInterface() {propertyValues {Curve {keys [
         {tangentMode=RCTM_User, time=0, value=1}
         {arriveTangent=-0.0405, interpMode=RCIM_Cubic, leaveTangent=-0.0405, tangentMode=RCTM_User, time=0.7009, value=1}
         {arriveTangent=-8.0089, interpMode=RCIM_Cubic, leaveTangent=-8.0089, tangentMode=RCTM_User, time=1, value=0}
        ]}}}
     Scale Curve=0.2
    )
    ScaleA=true
    ScaleRGB=true
   )
  ]
  rendererClasses [&NiagaraSpriteRendererProperties]
  renderers [{rendererClass=&NiagaraSpriteRendererProperties}]
 }
 Rings {
  emitterUpdateScript [
   &EmitterState(Life Cycle Mode=::NewEnumerator1, Loop Behavior=::NewEnumerator1, Loop Duration=5.1, UseLoopDelay=false)
   &SpawnRate(SpawnRate=2, Use Spawn Probability=false)
   &SpawnBurst_Instantaneous(Use Spawn Probability=false, UseLoopCountLimit=true)
  ]
  particleSpawnScript [
   &InitializeParticle(
    Color=&SelectLinearColorFromArray(
     Array Sampling Mode=::NewEnumerator0
     Direct Array Index=?
     Color Selection Array=DataInterface() {propertyValues {ColorData [<rgba 5 0.907 0 1>, <rgba 5 0.0914 0 1>, <rgba 5 2.3864 0 1>, <rgba 5 0.0914 0 1>, <rgba 5 0.0914 0 1>, <rgba 5 0.907 0 1>]}}
     Random Seed=0
    )
    Lifetime Max=1.5
    Lifetime Mode=::NewEnumerator1
    Mesh Scale Mode=::NewEnumerator1
    UsePositionOffset=false
    Write Material Random=true
   )
   &InitialMeshOrientation(Use Rotation Vector=false)
   /SetVariables(Particles.GlowColor=<rgba 1 0.117 0 0.5>) {isSetParameters=true}
   &SkeletalMeshLocation(
    Bone Sampling Mode=::NewEnumerator4
    MaskSpawnGroup=false
    Orientation Sampling=::NewEnumerator1
    Position Sampling=::NewEnumerator2
    Skeletal Mesh=$System.ScionLocation
   )
  ]
  particleUpdateScript [
   &ParticleState()
   &UpdateMeshOrientation(
    Rotation Rate=&FloatFromCurve(
     FloatCurve=DataInterface() {propertyValues {Curve {keys [
         {interpMode=RCIM_Cubic, leaveTangentWeight=1.7981, tangentMode=RCTM_User, time=0, value=2.3301}
         {arriveTangent=-0.1423, interpMode=RCIM_Cubic, leaveTangent=-0.1423, tangentMode=RCTM_User, time=5, value=0.4072}
        ]}}}
     CurveIndex=$Emitter.Age
     Scale Curve=1.6
    )
    Rotation Vector=<1 0 1>
   )
   &ScaleMeshSize(Scale Factor=&VectorFromFloat(Value=&FloatFromCurve(
      FloatCurve=DataInterface() {propertyValues {Curve {keys [
          <at 0 0.6>
          {
           arriveTangent=-0.0514
           arriveTangentWeight=0.0544
           interpMode=RCIM_Cubic
           leaveTangent=-0.0514
           leaveTangentWeight=0.0544
           tangentMode=RCTM_User
           tangentWeightMode=RCTWM_WeightedBoth
           time=0.015
           value=1.0988
          }
          {arriveTangentWeight=0.0276, interpMode=RCIM_Cubic, leaveTangentWeight=0.2941, tangentMode=RCTM_User, time=0.0623, value=0.7437}
          {arriveTangent=0.6783, interpMode=RCIM_Cubic, leaveTangent=0.6783, time=0.1173, value=0.8553}
          {arriveTangent=0.6257, interpMode=RCIM_Cubic, leaveTangent=0.6257, tangentMode=RCTM_User, time=0.379, value=0.9585}
          {arriveTangent=0.0516, interpMode=RCIM_Cubic, leaveTangent=0.0516, time=0.4944, value=0.8391}
          {arriveTangent=-1.1241, interpMode=RCIM_Cubic, leaveTangent=-1.1241, tangentMode=RCTM_User, time=0.7341, value=0.9768}
          {arriveTangentWeight=0.3078, interpMode=RCIM_Cubic, leaveTangentWeight=0.0104, tangentMode=RCTM_User, time=0.9, value=0.9}
          {arriveTangentWeight=0.1391, interpMode=RCIM_Cubic, leaveTangentWeight=0.1616, tangentMode=RCTM_User, time=0.96, value=1.6}
          <at 0.98 0 RCIM_Cubic>
         ]}}}
      CurveIndex=$Emitter.NormalizedLoopAge
      Scale Curve=0.6
     )))
   &DynamicMaterialParameters(
    Index 0 Param 1=&FloatFromCurve(
     FloatCurve=DataInterface() {propertyValues {Curve {keys [{interpMode=RCIM_Cubic, tangentMode=RCTM_User, time=0, value=1}, {interpMode=RCIM_Cubic, tangentMode=RCTM_User, time=1, value=0}]}}}
     Scale Curve=0
    )
    Index 1 Param 1=1
    Module.Index0Float0Write=true
    Module.Index0Float1Write=true
    Module.Index0Float2Write=true
    Module.Index0Float3Write=true
    Module.Index1Float0Write=true
    Module.Index1Float1Write=true
    Module.Index1Float2Write=true
    Module.Index1Float3Write=true
    Write Parameter Index 1=::NewEnumerator1
   )
   &ScaleColor(
    Scale Alpha=&FloatFromCurve(
     FloatCurve=DataInterface() {propertyValues {Curve {keys [
         {interpMode=RCIM_Cubic, leaveTangent=0.009, tangentMode=RCTM_Break, time=0.5705, value=1}
         {arriveTangent=-1.6389, interpMode=RCIM_Cubic, leaveTangent=-1.6389, tangentMode=RCTM_User, time=1, value=0}
        ]}}}
     Scale Curve=&FloatFromCurve(
      FloatCurve=DataInterface() {propertyValues {Curve {keys [
          {interpMode=RCIM_Cubic, leaveTangent=0.009, tangentMode=RCTM_Break, time=5, value=1}
          {arriveTangent=-10.5911, interpMode=RCIM_Cubic, leaveTangent=-10.5911, tangentMode=RCTM_User, time=5.1, value=0}
         ]}}}
      CurveIndex=$Emitter.Age
      Scale Curve=0.7
     )
    )
    Scale RGB=<1 1 2>
    ScaleA=true
    ScaleRGB=true
   )
   /SetVariables(Particles.GlowColor=&Multiply_LinearColorByFloat(
     Linear Color=$Particles.Initial.GlowColor
     Float=&FloatFromCurve(
      FloatCurve=DataInterface() {propertyValues {Curve {keys [
          {arriveTangent=1.2596, interpMode=RCIM_Cubic, leaveTangent=1.2596, tangentMode=RCTM_User, time=0, value=0.4}
          {arriveTangent=-0.4744, interpMode=RCIM_Cubic, leaveTangent=-6.1012, tangentMode=RCTM_Break, time=0.4, value=1}
          {arriveTangent=-0.0851, tangentMode=RCTM_User, time=1, value=0}
         ]}}}
      Scale Curve=&FloatFromCurve(
       FloatCurve=DataInterface() {propertyValues {Curve {keys [
           {interpMode=RCIM_Cubic, leaveTangent=0.009, tangentMode=RCTM_Break, time=5, value=1}
           {arriveTangent=-10.5911, interpMode=RCIM_Cubic, leaveTangent=-10.5911, tangentMode=RCTM_User, time=5.1, value=0}
          ]}}}
       CurveIndex=$Emitter.Age
      )
     )
    )) {isSetParameters=true}
   &SolveForcesAndVelocity(Clamp Velocity=false, Limit Acceleration=false)
  ]
  rendererClasses [&NiagaraMeshRendererProperties]
  renderers [
   {rendererClass=&NiagaraMeshRendererProperties, rendererIndex=0}
   {rendererClass=&NiagaraMeshRendererProperties, rendererIndex=1}
  ]
 }
 Runes {
  emitterUpdateScript [
   &EmitterState(Life Cycle Mode=::NewEnumerator1, Loop Behavior=::NewEnumerator1, Loop Duration=10, UseLoopDelay=false)
   &SpawnBurst_Instantaneous(Spawn Count=6, Use Spawn Probability=false, UseLoopCountLimit=false)
  ]
  particleSpawnScript [
   &PartitionParticles(Number Of Partitions=6)
   &InitializeParticle(
    Color=<rgba 0.9301 0.7682 0.1981 1>
    Lifetime=5.1
    Mesh Renderer Array Visibility Mode=::NewEnumerator1
    Mesh Renderer Info=DataInterface() {propertyValues {MeshRenderer=&self:Runes_0.NiagaraMeshRendererProperties_0}}
    Mesh Renderer Mesh Index:NiagaraInt32=$Particles.PartitionParticles.Partition
    Mesh Scale Mode=::NewEnumerator3
    Mesh Uniform Scale=0.7
    Sprite Size Mode=::NewEnumerator3
    Uniform Sprite Size=35
    UsePositionOffset=false
    Write Material Random=true
   )
  ]
  particleUpdateScript [
   &SocketLocation(
    SamplePosOffset=true
    Sampled Position Offset=<1 0 -1>
    Scale Sampling:ENiagara_AttributeSamplingApplyOutput=::NewEnumerator0
    Socket Index=$Particles.PartitionParticles.Partition
    Socket Reader=$System.RunesLocation
    Socket Sampling Mode=::NewEnumerator3
    VelocitySpeedLimit=false
   )
   &ParticleState()
   /SetVariables(
    Particles.GlowCameraOffset=10
    Particles.GlowColor=&MakeLinearColor(
     R:NiagaraFloat=0.6099
     G:NiagaraFloat=1
     B:NiagaraFloat=0.4223
     A:NiagaraFloat=&FloatFromCurve(
      FloatCurve=DataInterface() {propertyValues {Curve {keys [
          {tangentMode=RCTM_User, time=0, value=1}
          {arriveTangent=-0.0405, interpMode=RCIM_Cubic, leaveTangent=-0.0405, tangentMode=RCTM_User, time=0.7009, value=1}
          {arriveTangent=-8.0089, interpMode=RCIM_Cubic, leaveTangent=-8.0089, tangentMode=RCTM_User, time=1, value=0}
         ]}}}
      Scale Curve=0.2
     )
    )
    Particles.GlowSize=<35 35>
   ) {isSetParameters=true}
  ]
  rendererClasses [&NiagaraMeshRendererProperties, &NiagaraSpriteRendererProperties]
  renderers [
   {rendererClass=&NiagaraMeshRendererProperties, rendererIndex=0}
   {rendererClass=&NiagaraSpriteRendererProperties, rendererIndex=1}
  ]
 }
 Smoke {
  emitterUpdateScript [
   &EmitterState(Life Cycle Mode=::NewEnumerator1, Loop Behavior=::NewEnumerator1, Loop Delay=0.4, Loop Duration=4.95, UseLoopDelay=false)
   &SpawnRate(
    SpawnRate=&FloatFromCurve(
     FloatCurve=DataInterface() {propertyValues {Curve {keys [
         {interpMode=RCIM_Cubic, tangentMode=RCTM_User, time=0, value=1}
         {arriveTangent=-0.0787, interpMode=RCIM_Cubic, leaveTangent=-0.0787, tangentMode=RCTM_User, time=1, value=0}
        ]}}}
     CurveIndex=$Emitter.NormalizedLoopAge
     Scale Curve=5
    )
    Use Spawn Probability=false
   )
   &SpawnBurst_Instantaneous(Spawn Count=2, Spawn Group=1, Use Spawn Probability=false, UseLoopCountLimit=false)
  ]
  particleSpawnScript [
   &InitializeParticle(
    Color Maximum=<rgba 1 0.4174 0 1>
    Color Minimum=<rgba 1 0 0.0048 1>
    Color Mode=::NewEnumerator2
    Lifetime Max=2.5
    Lifetime Min=2.5
    Lifetime Mode=::NewEnumerator1
    Sprite Rotation Mode=::NewEnumerator2
    Sprite Size Mode=::NewEnumerator1
    Uniform Sprite Size Max=60
    Uniform Sprite Size Min=40
    UsePositionOffset=false
    Write Material Random=true
   )
   &ShapeLocation(MaskSpawnGroup=false, Override Seed=false, Sphere Radius=20, UseOwnerScale=true)
   &CameraOffset(Camera Offset Amount=30)
   /SetVariables(
    Particles.DarkCameraOffset:NiagaraFloat=&RandomRangeFloat(Minimum=0, Maximum=10)
    Particles.DarkColor:LinearColor=&FNTimeOfDayColor(
     Night Color=<rgba 0.2865 0.0579 0.0972 2>
     Morning Color=<rgba 0.4763 0.0494 0.1247 2>
     Day Color=<rgba 0.4763 0.0579 0.0073 2>
     Evening Color=<rgba 0.3003 0.0921 0.0669 2>
    )
   ) {isSetParameters=true}
  ]
  particleUpdateScript [
   &ParticleState()
   &SpriteRotationRate(Rotation Rate=&Multiply_Float(
     A=&RandomRangeFloat(Minimum=-50, Maximum=50)
     B=&FloatFromCurve(
      FloatCurve=DataInterface() {propertyValues {Curve {keys [
          {interpMode=RCIM_Cubic, tangentMode=RCTM_User, time=0, value=1}
          {arriveTangent=-3.5784, interpMode=RCIM_Cubic, leaveTangent=-3.5784, tangentMode=RCTM_User, time=1, value=0}
         ]}}}
      CurveIndex=$Emitter.Age
     )
    ))
   &ScaleSpriteSize(
    Scale Sprite Size Mode=::NewEnumerator0
    Uniform Scale Factor=&FloatFromCurve(FloatCurve=DataInterface() {propertyValues {Curve {keys [<at 0 0.8>,<at 1 1.2>]}}}, CurveIndex=$Particles.Age)
   )
   &SubUVAnimation(
    End Frame Range Override=30
    "Sprite Renderer "=DataInterface() {propertyValues {SpriteRenderer=&self:Smoke_0.NiagaraSpriteRendererProperties_0}}
    Start Frame Range Override=&RandomRangeInt(Minimum=0, Maximum=5)
    SubUV Lookup Index Scale=2.5
    UseEndFrame=true
    UseStartFrame=true
   )
   &ScaleColor(
    Scale Alpha=&Multiply_Float(
     A=&RandomRangeFloat(Minimum=0.5, Maximum=0.3)
     B=&MaskFloatBySpawnGroup(
      Masked Float=&FloatFromCurve(
       FloatCurve=DataInterface() {propertyValues {Curve {keys [
           {interpMode=RCIM_Cubic, tangentMode=RCTM_User, time=0, value=0}
           {interpMode=RCIM_Cubic, tangentMode=RCTM_User, time=0.5, value=1}
           {tangentMode=RCTM_User, time=1, value=0}
          ]}}}
       CurveIndex=$Particles.Age
       Scale Curve=0.3
      )
      Passthrough Float=&FloatFromCurve(
       FloatCurve=DataInterface() {propertyValues {Curve {keys [
           {interpMode=RCIM_Cubic, tangentMode=RCTM_User, time=0, value=1}
           {interpMode=RCIM_Cubic, tangentMode=RCTM_User, time=0.5, value=1}
           {tangentMode=RCTM_User, time=1, value=0}
          ]}}}
       CurveIndex=$Particles.Age
      )
     )
    )
    ScaleA=true
    ScaleRGB=true
   )
   &SolveForcesAndVelocity(Clamp Velocity=false, Limit Acceleration=false)
   /SetVariables(
    Particles.DarkColor:LinearColor=&MakeLinearColor(
     R:NiagaraFloat=&FNTimeOfDayScalar(Night Scale=0.286, Morning Scale=0.476, Day Scale=0.476, Evening Scale=0.3)
     G:NiagaraFloat=&FNTimeOfDayScalar(Night Scale=0.058, Morning Scale=0.049, Day Scale=0.058, Evening Scale=0.092)
     B:NiagaraFloat=&FNTimeOfDayScalar(Night Scale=0.097, Morning Scale=0.125, Day Scale=0.007, Evening Scale=0.067)
     A:NiagaraFloat=&FloatFromCurve(
      FloatCurve=DataInterface() {propertyValues {Curve {keys [
          {interpMode=RCIM_Cubic, tangentMode=RCTM_User, time=0, value=0}
          {arriveTangent=-2.0189, interpMode=RCIM_Cubic, leaveTangent=-2.1015, tangentMode=RCTM_Break, time=0.2, value=0.9}
          {arriveTangent=-1.1521, interpMode=RCIM_Cubic, tangentMode=RCTM_Break, time=1, value=0}
         ]}}}
      Scale Curve=&FloatFromCurve(
       FloatCurve=DataInterface() {propertyValues {Curve {keys [{interpMode=RCIM_Cubic, tangentMode=RCTM_User, time=4.5123, value=1}, {interpMode=RCIM_Cubic, tangentMode=RCTM_User, time=4.95, value=0}]}}}
       CurveIndex=$Emitter.Age
       Scale Curve=2
      )
     )
    )
    Particles.DarkRotation:NiagaraFloat=&RandomRangeFloat(Minimum=-150, Maximum=150)
    Particles.DarkSize:Vector2f=&Multiply_Vector2D_ByFloat(
     Float=&FloatFromCurve(FloatCurve=DataInterface() {propertyValues {Curve {keys [
          {arriveTangent=1.6237, interpMode=RCIM_Cubic, leaveTangent=1.6237, tangentMode=RCTM_User, time=0, value=1}
          {arriveTangent=0.637, tangentMode=RCTM_User, time=1, value=1.7}
         ]}}})
     Vector2D=&RandomRangeVector2D(Minimum=<40 40>, Maximum=<80 80>)
    )
   ) {isSetParameters=true}
  ]
  rendererClasses [&NiagaraSpriteRendererProperties]
  renderers [
   {rendererClass=&NiagaraSpriteRendererProperties}
   {rendererClass=&NiagaraSpriteRendererProperties, rendererIndex=1}
  ]
 }
 SmokeCircle {
  emitterUpdateScript [
   &EmitterState(Life Cycle Mode=::NewEnumerator1, Loop Behavior=::NewEnumerator1, Loop Delay=5.05, Loop Duration=2, UseLoopDelay=true)
   &SpawnBurst_Instantaneous(Spawn Count=3, Use Spawn Probability=false, UseLoopCountLimit=false)
  ]
  particleSpawnScript [
   /SetVariables(Particles.RandomRange:NiagaraFloat=&Multiply_Float(
     A=&RandomRangeFloat(Minimum=0.25, Maximum=0.375)
     B=&FloatFromCurve(
      FloatCurve=DataInterface() {propertyValues {Curve {keys [<at 0 1 RCIM_Cubic>, {arriveTangent=7.3942, interpMode=RCIM_Cubic, leaveTangent=7.3942, tangentMode=RCTM_User, time=0.3, value=2}]}}}
      CurveIndex=$Emitter.Age
     )
    )) {isSetParameters=true}
   &InitializeParticle(
    Color=&FNTimeOfDayColor(
     Night Color=&SelectLinearColorFromArray(
      Array Sampling Mode=::NewEnumerator0
      Direct Array Index=?
      Color Selection Array=DataInterface() {propertyValues {ColorData [<rgba 0.1355 0.1754 0.2587 1>, <rgba 0.3111 0.387 0.5457 1>]}}
      Random Seed=0
     )
     Morning Color=&SelectLinearColorFromArray(
      Array Sampling Mode=::NewEnumerator0
      Direct Array Index=?
      Color Selection Array=DataInterface() {propertyValues {ColorData [<rgba 0.3189 0.128 0.1426 1>, <rgba 0.5851 0.2349 0.2616 1>]}}
      Random Seed=0
     )
     Day Color=&SelectLinearColorFromArray(
      Array Sampling Mode=::NewEnumerator0
      Direct Array Index=?
      Color Selection Array=DataInterface() {propertyValues {ColorData [<rgba 0.3189 0.1904 0.1111 1>, <rgba 0.7633 0.4558 0.266 1>]}}
      Random Seed=0
     )
     Evening Color=&SelectLinearColorFromArray(
      Array Sampling Mode=::NewEnumerator0
      Direct Array Index=?
      Color Selection Array=DataInterface() {propertyValues {ColorData [<rgba 0.2564 0.2542 0.1823 1>, <rgba 0.6128 0.6076 0.4358 1>]}}
      Random Seed=0
     )
    )
    Lifetime=0.6
    Mesh Scale=&Multiply_VectorByFloat(Vector=<1 1.5 1>, Float=0.25)
    Mesh Scale Mode=::NewEnumerator4
    UsePositionOffset=false
    Write Material Random=true
   )
   &CameraOffset(Camera Offset Amount=15)
   &InitialMeshOrientation(
    Mesh Axis To Orient=<0 0 0>
    Mesh Orientation Mode=::NewEnumerator2
    Orientation Vector=<0 0 1>
    Rotation=&RandomRangeVector(Minimum=<0 0 0>, Maximum=<0 1 0>)
    Use Rotation Vector=true
   )
   &ShapeLocation(MaskSpawnGroup=false, Override Seed=false, Sphere Radius=5, UseOwnerScale=true)
  ]
  particleUpdateScript [
   &ParticleState()
   &ScaleMeshSize(Scale Factor=&VectorFromFloat(Value=&FloatFromCurve(
      FloatCurve=DataInterface() {propertyValues {Curve {keys [
          {arriveTangent=5.6347, interpMode=RCIM_Cubic, leaveTangent=5.6347, tangentMode=RCTM_User, time=0, value=0.3788}
          {arriveTangent=1.1451, interpMode=RCIM_Cubic, leaveTangent=1.1451, time=0.2, value=0.85}
          {arriveTangent=0.2133, interpMode=RCIM_Cubic, leaveTangent=0.2133, tangentMode=RCTM_User, time=0.6186, value=1.0872}
          <at 0.9985 1.2296 RCIM_Cubic>
         ]}}}
      Scale Curve=0.95
     )))
   &DynamicMaterialParameters(
    Index 0 Param 1=&FloatFromCurve(FloatCurve=DataInterface() {propertyValues {Curve {keys [
         {arriveTangent=2.3202, interpMode=RCIM_Cubic, leaveTangent=2.3202, tangentMode=RCTM_User, time=0, value=0.2481}
         {arriveTangent=0.6377, interpMode=RCIM_Cubic, leaveTangent=0.6377, time=0.0995, value=0.4178}
         {interpMode=RCIM_Cubic, tangentMode=RCTM_User, time=0.3951, value=0.5}
         {arriveTangent=-1.1781, tangentMode=RCTM_User, time=1, value=0}
        ]}}})
    Module.Index0Param0Write=true
    Module.Index0Param1Write=false
    Module.Index0Param2Write=false
    Module.Index0Param3Write=false
   )
   &ScaleColor(
    Scale Alpha=&FloatFromCurve(
     FloatCurve=DataInterface() {propertyValues {Curve {keys [
         {arriveTangent=-0.0871, interpMode=RCIM_Cubic, leaveTangent=-0.0871, tangentMode=RCTM_User, time=0, value=0.6231}
         {interpMode=RCIM_Cubic, tangentMode=RCTM_User, time=0.5, value=0.7329}
         {tangentMode=RCTM_User, time=1, value=0}
        ]}}}
     Scale Curve=&FloatFromCurve(
      FloatCurve=DataInterface() {propertyValues {Curve {keys [
          {interpMode=RCIM_Cubic, tangentMode=RCTM_User, time=0, value=1}
          {arriveTangent=-0.0131, interpMode=RCIM_Cubic, leaveTangent=0.1095, tangentMode=RCTM_Break, time=5.3442, value=1}
          {arriveTangent=-3.0223, interpMode=RCIM_Cubic, leaveTangent=-3.0223, tangentMode=RCTM_User, time=5.65, value=0}
         ]}}}
      CurveIndex=$Emitter.Age
      Scale Curve=0.5
     )
    )
    Scale RGB=&MakeVectorFromLinearColorRGB(Color=&ColorFromCurve(ColorCurve=DataInterface() {propertyValues {
        AlphaCurve {keys []}
        BlueCurve {keys [
          <at 0 1 RCIM_Cubic>
          {arriveTangent=-1.2161, interpMode=RCIM_Cubic, leaveTangent=-1.2161, time=0.1313, value=1}
          <at 0.4231 0.4855>
          <at 1 0.2309>
         ]}
        GreenCurve {keys [
          <at 0 1 RCIM_Cubic>
          {arriveTangent=-1.2767, interpMode=RCIM_Cubic, leaveTangent=-1.2767, time=0.1313, value=0.8855}
          <at 0.4231 0.4599>
          <at 1 0.2157>
         ]}
        RedCurve {keys [
          <at 0 1 RCIM_Cubic>
          {arriveTangent=-1.3057, interpMode=RCIM_Cubic, leaveTangent=-1.3057, time=0.1313, value=0.8293}
          <at 0.4231 0.4476>
          <at 1 0.2075>
         ]}
       }}))
    ScaleA=true
    ScaleRGB=true
   )
   &UpdateMeshOrientation(Rotation Rate=0.2, Rotation Vector=<0 1 0>) {module=UpdateMeshOrientation001}
  ]
  rendererClasses [&NiagaraMeshRendererProperties]
  renderers [{rendererClass=&NiagaraMeshRendererProperties, rendererIndex=0}]
 }
}
summary {
 systemName=NS_Emote_GnocchiTea_Burst
 emitters [
  {emitterName=Light, rendererClasses [&NiagaraSpriteRendererProperties, &NiagaraLightRendererProperties]}
  {emitterName=Smoke, rendererClasses [&NiagaraSpriteRendererProperties]}
  {emitterName=DustBursts, rendererClasses [&NiagaraSpriteRendererProperties]}
  {emitterName=Chunks001, simTarget=GPUComputeSim, rendererClasses [&NiagaraSpriteRendererProperties]}
  {emitterName=SmokeCircle, rendererClasses [&NiagaraMeshRendererProperties]}
  {emitterName=Rings, rendererClasses [&NiagaraMeshRendererProperties]}
  {emitterName=PieceGlows, rendererClasses [&NiagaraSpriteRendererProperties]}
  {emitterName=GemGlows001, rendererClasses [&NiagaraSpriteRendererProperties]}
  {emitterName=Runes, rendererClasses [&NiagaraMeshRendererProperties, &NiagaraSpriteRendererProperties]}
 ]
 userVariables [
  User.bIsFrontEnd {type=NiagaraBool}
  User.bIsFrontEndPreview {type=NiagaraBool}
 ]
}
