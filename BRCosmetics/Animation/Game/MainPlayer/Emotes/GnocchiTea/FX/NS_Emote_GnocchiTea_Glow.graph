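# Niagara effect for the GnocchiTea player emote, layering gem glows, short light flashes, smoke, rings, dust and spinning motion-blur meshes.
# Gem emitters use particle partitions to follow three skeletal mesh bones, with spawn groups splitting green and orange pairs by color and offset.
# MotionBlur meshes track a separate ScionLocation mesh, with age-driven spin and fade.
# Time-of-day functions adjust glow brightness and ring and dust colors.
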NiagaraSystem /BRCosmetics/Animation/Game/MainPlayer/Emotes/GnocchiTea/FX/NS_Emote_GnocchiTea_Glow
dependencies {
 usedDataInterfaces [
  &NiagaraDataInterfaceCurve
  &NiagaraDataInterfaceVectorCurve
  &NiagaraDataInterfaceVector2DCurve
  &NiagaraDataInterfaceSpriteRendererInfo
  &NiagaraDataInterfaceArrayColor
  &NiagaraDataInterfaceColorCurve
 ]
 usedDynamicInputs [&Vector2DFromCurve, &RandomRangeVector2D]
 usedModules [
  &self:SystemSpawnScript.SystemScriptSource.SystemScriptGraph.NiagaraNodeAssignment_0.SetVariables_#1
  &self:SystemSpawnScript.SystemScriptSource.SystemScriptGraph.NiagaraNodeAssignment_1.SetVariables_#2
  &SystemState
 ]
}
emitters {
 GemGlows {
  emitterUpdateScript [
   &EmitterState(Life Cycle Mode=::NewEnumerator1, Loop Behavior=::NewEnumerator1, Loop Duration=10, UseLoopDelay=false)
   &SpawnBurst_Instantaneous(Spawn Count=3, Spawn Group=1, Use Spawn Probability=false, UseLoopCountLimit=false) {module=SpawnBurst_Instantaneous001}
   &SpawnBurst_Instantaneous(Spawn Count=3, Use Spawn Probability=false, UseLoopCountLimit=false)
  ]
  particleSpawnScript [
   &PartitionParticles(Number Of Partitions=3)
   &FN_SetSpawnGroup(Number of Spawn Groups=2)
   &InitializeParticle(
    Color=&MaskLinearColorBySpawnGroup(Masked Linear Color=<rgba 0.1181 1 0.3548 1>, Passthrough Linear Color=<rgba 1 0.3847 0 0.4493>)
    Lifetime=10
    Sprite Size Mode=::NewEnumerator3
    Uniform Sprite Size=17
    UsePositionOffset=false
    Write Material Random=true
   )
   &CameraOffset(Camera Offset Amount=10)
  ]
  particleUpdateScript [
   &ParticleState()
   &SkeletalMeshLocation(
    Bone / Socket Index=$Particles.PartitionParticles.Partition
    Bone Sampling Mode=::NewEnumerator4
    MaskSpawnGroup=false
    Sampled Position Offset=&MaskVectorBySpawnGroup(Masked Vector=<0 -6 0>, Passthrough Vector=<0 0 0>)
    Skeletal Mesh=$System.Skeletal Mesh
   )
   &ScaleColor(
    Scale Alpha=&FloatFromCurve(
     FloatCurve=DataInterface() {propertyValues {Curve {keys [
         {arriveTangentWeight=0.0665, interpMode=RCIM_Cubic, leaveTangentWeight=0.0103, tangentMode=RCTM_User, time=0.1485, value=6}
         {
          arriveTangent=-0.1272
          arriveTangentWeight=1.3428
          interpMode=RCIM_Cubic
          leaveTangent=-0.1272
          tangentMode=RCTM_User
          time=0.262
          value=0.8567
         }
         {
          arriveTangent=0.4607
          arriveTangentWeight=1.1962
          interpMode=RCIM_Cubic
          leaveTangent=0.4607
          leaveTangentWeight=0.0691
          tangentMode=RCTM_User
          time=1.0511
          value=1.1243
         }
         {
          arriveTangent=4.7679
          arriveTangentWeight=1.263
          interpMode=RCIM_Cubic
          leaveTangent=-0.011
          leaveTangentWeight=0.0345
          tangentMode=RCTM_Break
          time=2.853
          value=4
         }
         {arriveTangentWeight=0.8914, interpMode=RCIM_Cubic, leaveTangentWeight=0.4457, tangentMode=RCTM_User, time=3.6777, value=4}
         {arriveTangentWeight=0.047, interpMode=RCIM_Cubic, leaveTangentWeight=1.9605, tangentMode=RCTM_User, time=3.8, value=1.4328}
         {arriveTangent=-0.523, interpMode=RCIM_Cubic, leaveTangent=-0.523, tangentMode=RCTM_User, time=8.9, value=0}
        ]}}}
     CurveIndex=$Particles.Age
     Scale Curve=&MaskFloatBySpawnGroup(
      Masked Float=&FNTimeOfDayScalar(Night Scale=0.1, Morning Scale=0.3, Day Scale=0.3, Evening Scale=0.3)
      Passthrough Float=&FNTimeOfDayScalar(Night Scale=0.7, Morning Scale=1, Evening Scale=1)
     )
    )
    ScaleA=true
    ScaleRGB=true
   )
  ]
  rendererClasses [&NiagaraSpriteRendererProperties]
  renderers [{rendererClass=&NiagaraSpriteRendererProperties}]
 }
 GemLights {
  emitterUpdateScript [
   &EmitterState(Life Cycle Mode=::NewEnumerator1, Loop Behavior=::NewEnumerator1, Loop Duration=4, UseLoopDelay=false)
   &SpawnBurst_Instantaneous(Spawn Count=3, Spawn Group=1, Use Spawn Probability=false, UseLoopCountLimit=false)
   &SpawnBurst_Instantaneous(Spawn Count=3, Use Spawn Probability=false, UseLoopCountLimit=false) {module=SpawnBurst_Instantaneous001}
  ]
  particleSpawnScript [
   &PartitionParticles(Number Of Partitions=3)
   &FN_SetSpawnGroup(Number of Spawn Groups=2)
   &InitializeParticle(
    Color=&MaskLinearColorBySpawnGroup(Masked Linear Color=<rgba 0.1181 1 0.3548 1>, Passthrough Linear Color=<rgba 1 0.3748 0 1>)
    Lifetime=0.3
    Mesh Scale Mode=::NewEnumerator3
    Mesh Uniform Scale=0.1
    Sprite Size Mode=::NewEnumerator3
    Uniform Sprite Size=20
    UsePositionOffset=false
    Write Material Random=true
   )
   &CameraOffset(Camera Offset Amount=20) {enabled=false}
   &InitialMeshOrientation(
    Mesh Axis To Orient=<0 1 0>
    Mesh Orientation Mode=::NewEnumerator2
    Orientation Vector=&MaskVectorBySpawnGroup(Masked Vector=<0 0 -1>, Passthrough Vector=<0 0 1>)
    Use Rotation Vector=false
   )
  ]
  particleUpdateScript [
   &ParticleState()
   &ScaleMeshSize(Scale Factor=&VectorFromCurve(VectorCurve=DataInterface() {propertyValues {
       XCurve {keys [{interpMode=RCIM_Cubic, tangentMode=RCTM_User, time=0.5, value=1}, {interpMode=RCIM_Cubic, tangentMode=RCTM_User, time=1, value=1}]}
       YCurve {keys [{interpMode=RCIM_Cubic, tangentMode=RCTM_User, time=1, value=1}]}
       ZCurve {keys [
         <at 0 0.8 RCIM_Cubic>
         {arriveTangentWeight=0.0359, interpMode=RCIM_Cubic, leaveTangentWeight=0.1436, tangentMode=RCTM_User, time=0.1206, value=1.326}
         {
          arriveTangent=-1.883
          arriveTangentWeight=0.0299
          interpMode=RCIM_Cubic
          leaveTangent=-1.883
          leaveTangentWeight=0.2006
          tangentMode=RCTM_User
          time=0.2069
          value=1
         }
         {interpMode=RCIM_Cubic, tangentMode=RCTM_User, time=1, value=0}
        ]}
      }}))
   &SkeletalMeshLocation(
    Bone / Socket Index=$Particles.PartitionParticles.Partition
    Bone Sampling Mode=::NewEnumerator4
    MaskSpawnGroup=false
    Sampled Position Offset=&MaskVectorBySpawnGroup(Masked Vector=<0 -4 0>, Passthrough Vector=<0 0 0>)
    Skeletal Mesh=$System.Skeletal Mesh
   )
   &ScaleColor(
    Scale Alpha=&FloatFromCurve(FloatCurve=DataInterface() {propertyValues {Curve {keys [
         {interpMode=RCIM_Cubic, tangentMode=RCTM_User, time=0, value=1}
         {arriveTangent=-14.0771, interpMode=RCIM_Cubic, leaveTangent=-14.0771, tangentMode=RCTM_User, time=0.0532, value=1.9294}
         {arriveTangent=-2.4966, interpMode=RCIM_Cubic, leaveTangent=-2.4966, time=0.2, value=0.5}
         {arriveTangent=-0.0083, interpMode=RCIM_Cubic, leaveTangent=-0.0083, tangentMode=RCTM_User, time=0.826, value=0}
        ]}}})
    ScaleA=true
    ScaleRGB=true
   )
  ]
  rendererClasses [&NiagaraMeshRendererProperties]
  renderers [{rendererClass=&NiagaraMeshRendererProperties, rendererIndex=0}]
 }
 LightSmoke {
  emitterUpdateScript [
   &EmitterState(Life Cycle Mode=::NewEnumerator1, Loop Behavior=::NewEnumerator1, Loop Delay=0.4, Loop Duration=4, UseLoopDelay=false)
   &SpawnRate(
    SpawnRate=&FloatFromCurve(
     FloatCurve=DataInterface() {propertyValues {Curve {keys [
         {arriveTangent=1.4946, interpMode=RCIM_Cubic, leaveTangent=1.4946, tangentMode=RCTM_User, time=0, value=0}
         {interpMode=RCIM_Cubic, tangentMode=RCTM_User, time=1, value=1}
        ]}}}
     CurveIndex=$Emitter.NormalizedLoopAge
     Scale Curve=4
    )
    Use Spawn Probability=false
   )
   &SpawnBurst_Instantaneous(Spawn Count=2, Spawn Group=1, Use Spawn Probability=false, UseLoopCountLimit=false) {enabled=false}
  ]
  particleSpawnScript [
   &InitializeParticle(
    Color Maximum=<rgba 1 0.4174 0 1>
    Color Minimum=<rgba 0.4022 0.0095 0.0114 1>
    Color Mode=::NewEnumerator2
    Lifetime Min=3
    Lifetime Mode=::NewEnumerator1
    Sprite Rotation Mode=::NewEnumerator2
    Sprite Size Mode=::NewEnumerator1
    Uniform Sprite Size Max=60
    Uniform Sprite Size Min=40
    UsePositionOffset=false
    Write Material Random=true
   )
   &ShapeLocation(MaskSpawnGroup=false, Override Seed=false, Sphere Radius=20, UseOwnerScale=true)
   &CameraOffset(Camera Offset Amount=30)
  ]
  particleUpdateScript [
   &ParticleState()
   &SpriteRotationRate(Rotation Rate=&Multiply_Float(
     A=&RandomRangeFloat(Minimum=-50, Maximum=50)
     B=&FloatFromCurve(
      FloatCurve=DataInterface() {propertyValues {Curve {keys [
          {interpMode=RCIM_Cubic, tangentMode=RCTM_User, time=0, value=1}
          {arriveTangent=-3.5784, interpMode=RCIM_Cubic, leaveTangent=-3.5784, tangentMode=RCTM_User, time=1, value=0}
         ]}}}
      CurveIndex=$Emitter.Age
     )
    ))
   &ScaleSpriteSize(
    Scale Sprite Size Mode=::NewEnumerator0
    Uniform Scale Factor=&FloatFromCurve(FloatCurve=DataInterface() {propertyValues {Curve {keys [
         {interpMode=RCIM_Cubic, tangentMode=RCTM_User, time=0, value=0}
         {arriveTangent=1.5912, interpMode=RCIM_Cubic, leaveTangent=1.5912, tangentMode=RCTM_User, time=1, value=1}
        ]}}})
   )
   &SubUVAnimation(
    End Frame Range Override=30
    "Sprite Renderer "=DataInterface() {propertyValues {SpriteRenderer=&self:LightSmoke_0.NiagaraSpriteRendererProperties_0}}
    Start Frame Range Override=&RandomRangeInt(Minimum=0, Maximum=5)
    UseEndFrame=true
    UseStartFrame=true
   )
   &ScaleColor(
    Scale Alpha=&Multiply_Float(
     A=&RandomRangeFloat(Minimum=0.5, Maximum=0.3)
     B=&MaskFloatBySpawnGroup(
      Masked Float=&FloatFromCurve(
       FloatCurve=DataInterface() {propertyValues {Curve {keys [
           {interpMode=RCIM_Cubic, tangentMode=RCTM_User, time=0, value=0}
           {interpMode=RCIM_Cubic, tangentMode=RCTM_User, time=0.5, value=1}
           {tangentMode=RCTM_User, time=1, value=0}
          ]}}}
       Scale Curve=0.3
      )
      Passthrough Float=&FloatFromCurve(
       FloatCurve=DataInterface() {propertyValues {Curve {keys [<at 0 0 RCIM_Cubic>, {arriveTangent=2.1748, interpMode=RCIM_Cubic, leaveTangent=2.1748, tangentMode=RCTM_User, time=1.0244, value=1}]}}}
       CurveIndex=$Emitter.Age
      )
     )
    )
    ScaleA=true
    ScaleRGB=true
   )
   &SolveForcesAndVelocity(Clamp Velocity=false, Limit Acceleration=false)
  ]
  rendererClasses [&NiagaraSpriteRendererProperties]
  renderers [{rendererClass=&NiagaraSpriteRendererProperties}]
 }
 Main_Ring {
  emitterUpdateScript [
   &EmitterState(Life Cycle Mode=::NewEnumerator1, Loop Behavior=::NewEnumerator1, Loop Delay=0.01, Loop Duration=1, UseLoopDelay=false)
   &SpawnBurst_Instantaneous(Spawn Count=3, Use Spawn Probability=false, UseLoopCountLimit=false)
  ]
  particleSpawnScript [
   /SetVariables(Particles.RandomRange:NiagaraFloat=&Multiply_Float(
     A=&RandomRangeFloat(Minimum=0.25, Maximum=0.375)
     B=&FloatFromCurve(
      FloatCurve=DataInterface() {propertyValues {Curve {keys [<at 0 1 RCIM_Cubic>, {arriveTangent=7.3942, interpMode=RCIM_Cubic, leaveTangent=7.3942, tangentMode=RCTM_User, time=0.3, value=2}]}}}
      CurveIndex=$Emitter.Age
     )
    )) {isSetParameters=true}
   &InitializeParticle(
    Color=&FNTimeOfDayColor(
     Night Color=&SelectLinearColorFromArray(
      Array Sampling Mode=::NewEnumerator0
      Direct Array Index=?
      Color Selection Array=DataInterface() {propertyValues {ColorData [<rgba 0.1355 0.1754 0.2587 1>, <rgba 0.3111 0.387 0.5457 1>]}}
      Random Seed=0
     )
     Morning Color=&SelectLinearColorFromArray(
      Array Sampling Mode=::NewEnumerator0
      Direct Array Index=?
      Color Selection Array=DataInterface() {propertyValues {ColorData [<rgba 0.3189 0.128 0.1426 1>, <rgba 0.5851 0.2349 0.2616 1>]}}
      Random Seed=0
     )
     Day Color=&SelectLinearColorFromArray(
      Array Sampling Mode=::NewEnumerator0
      Direct Array Index=?
      Color Selection Array=DataInterface() {propertyValues {ColorData [<rgba 0.3189 0.1904 0.1111 1>, <rgba 0.7633 0.4558 0.266 1>]}}
      Random Seed=0
     )
     Evening Color=&SelectLinearColorFromArray(
      Array Sampling Mode=::NewEnumerator0
      Direct Array Index=?
      Color Selection Array=DataInterface() {propertyValues {ColorData [<rgba 0.2564 0.2542 0.1823 1>, <rgba 0.6128 0.6076 0.4358 1>]}}
      Random Seed=0
     )
    )
    Lifetime=0.4
    Mesh Scale=&Multiply_VectorByFloat(Vector=<1 1.5 1>, Float=0.25)
    Mesh Scale Mode=::NewEnumerator4
    UsePositionOffset=false
    Write Material Random=true
   )
   &CameraOffset(Camera Offset Amount=15)
   &InitialMeshOrientation(
    Mesh Axis To Orient=<0 0 0>
    Mesh Orientation Mode=::NewEnumerator2
    Orientation Vector=<0 0 1>
    Rotation=&RandomRangeVector(Minimum=<0 0 0>, Maximum=<0 1 0>)
    Use Rotation Vector=true
   )
   &ShapeLocation(MaskSpawnGroup=false, Override Seed=false, Sphere Radius=5, UseOwnerScale=true)
  ]
  particleUpdateScript [
   &ParticleState()
   &ScaleMeshSize(Scale Factor=&VectorFromFloat(Value=&FloatFromCurve(
      FloatCurve=DataInterface() {propertyValues {Curve {keys [
          {arriveTangent=5.6347, interpMode=RCIM_Cubic, leaveTangent=5.6347, tangentMode=RCTM_User, time=0, value=0.3788}
          {arriveTangent=1.1451, interpMode=RCIM_Cubic, leaveTangent=1.1451, time=0.2, value=0.85}
          {arriveTangent=0.2133, interpMode=RCIM_Cubic, leaveTangent=0.2133, tangentMode=RCTM_User, time=0.6186, value=1.0872}
          <at 0.9985 1.2296 RCIM_Cubic>
         ]}}}
      Scale Curve=0.95
     )))
   &DynamicMaterialParameters(
    Index 0 Param 1=&FloatFromCurve(FloatCurve=DataInterface() {propertyValues {Curve {keys [
         {arriveTangent=2.3202, interpMode=RCIM_Cubic, leaveTangent=2.3202, tangentMode=RCTM_User, time=0, value=0.2481}
         {arriveTangent=0.6377, interpMode=RCIM_Cubic, leaveTangent=0.6377, time=0.0995, value=0.4178}
         {interpMode=RCIM_Cubic, tangentMode=RCTM_User, time=0.3951, value=0.5}
         {arriveTangent=-1.1781, tangentMode=RCTM_User, time=1, value=0}
        ]}}})
    Module.Index0Param0Write=true
    Module.Index0Param1Write=false
    Module.Index0Param2Write=false
    Module.Index0Param3Write=false
   )
   &ScaleColor(
    Scale Alpha=&FloatFromCurve(
     FloatCurve=DataInterface() {propertyValues {Curve {keys [
         {arriveTangent=-0.0871, interpMode=RCIM_Cubic, leaveTangent=-0.0871, tangentMode=RCTM_User, time=0, value=0.6231}
         {interpMode=RCIM_Cubic, tangentMode=RCTM_User, time=0.5, value=0.7329}
         {tangentMode=RCTM_User, time=1, value=0}
        ]}}}
     Scale Curve=&FloatFromCurve(
      FloatCurve=DataInterface() {propertyValues {Curve {keys [
          {interpMode=RCIM_Cubic, tangentMode=RCTM_User, time=0, value=1}
          {arriveTangent=-0.0131, interpMode=RCIM_Cubic, leaveTangent=0.1095, tangentMode=RCTM_Break, time=0.3, value=1}
          {arriveTangent=-3.0223, interpMode=RCIM_Cubic, leaveTangent=-3.0223, tangentMode=RCTM_User, time=0.7, value=0}
         ]}}}
      CurveIndex=$Emitter.Age
      Scale Curve=0.4
     )
    )
    Scale RGB=&MakeVectorFromLinearColorRGB(Color=&ColorFromCurve(ColorCurve=DataInterface() {propertyValues {
        AlphaCurve {keys []}
        BlueCurve {keys [
          <at 0 1 RCIM_Cubic>
          {arriveTangent=-1.2161, interpMode=RCIM_Cubic, leaveTangent=-1.2161, time=0.1313, value=1}
          <at 0.4231 0.4855>
          <at 1 0.2309>
         ]}
        GreenCurve {keys [
          <at 0 1 RCIM_Cubic>
          {arriveTangent=-1.2767, interpMode=RCIM_Cubic, leaveTangent=-1.2767, time=0.1313, value=0.8855}
          <at 0.4231 0.4599>
          <at 1 0.2157>
         ]}
        RedCurve {keys [
          <at 0 1 RCIM_Cubic>
          {arriveTangent=-1.3057, interpMode=RCIM_Cubic, leaveTangent=-1.3057, time=0.1313, value=0.8293}
          <at 0.4231 0.4476>
          <at 1 0.2075>
         ]}
       }}))
    ScaleA=true
    ScaleRGB=true
   )
   &UpdateMeshOrientation(Rotation Rate=0.2, Rotation Vector=<0 1 0>) {module=UpdateMeshOrientation001}
  ]
  rendererClasses [&NiagaraMeshRendererProperties]
  renderers [{rendererClass=&NiagaraMeshRendererProperties, rendererIndex=0}]
 }
 MotionBlur {
  emitterUpdateScript [
   &EmitterState(Life Cycle Mode=::NewEnumerator1, Loop Behavior=::NewEnumerator1, Loop Delay=0.6, Loop Duration=5.1, UseLoopDelay=true)
   &SpawnBurst_Instantaneous(Spawn Count=2, Use Spawn Probability=false, UseLoopCountLimit=true)
  ]
  particleSpawnScript [
   &FN_SetSpawnGroup(Number of Spawn Groups=2)
   &InitializeParticle(
    Color=&MaskLinearColorBySpawnGroup(Masked Linear Color=<rgba 0.1181 1 0.3548 1>, Passthrough Linear Color=<rgba 1 0.3847 0 1>)
    Lifetime=3
    Mesh Scale Mode=::NewEnumerator3
    Mesh Uniform Scale=&Multiply_Float(A=1.5, B=0.6)
    UsePositionOffset=false
    Write Material Random=true
   )
   &InitialMeshOrientation(Mesh Orientation Mode=::NewEnumerator2, Orientation Vector=<0 1 0>, Use Rotation Vector=false)
  ]
  particleUpdateScript [
   &SkeletalMeshLocation(
    Bone Sampling Mode=::NewEnumerator4
    MaskSpawnGroup=false
    Orientation Sampling=::NewEnumerator1
    Sampled Position Offset=&MaskVectorBySpawnGroup(Masked Vector=<0 0 0>, Passthrough Vector=<0 10 0>)
    Skeletal Mesh=$System.ScionLocation
   )
   &ParticleState()
   &UpdateMeshOrientation(Rotation Rate=&FloatFromCurve(
     FloatCurve=DataInterface() {propertyValues {Curve {keys [
         {arriveTangent=-6.7937, interpMode=RCIM_Cubic, leaveTangent=4.4572, tangentMode=RCTM_Break, time=0, value=-2}
         {arriveTangent=0.3856, tangentMode=RCTM_User, time=2.6256, value=2}
        ]}}}
     CurveIndex=$Emitter.Age
     Scale Curve=1.6
    ))
   &DynamicMaterialParameters(
    Index 0 Param 1=&FloatFromCurve(
     FloatCurve=DataInterface() {propertyValues {Curve {keys [{interpMode=RCIM_Cubic, tangentMode=RCTM_User, time=0, value=1}, {interpMode=RCIM_Cubic, tangentMode=RCTM_User, time=1, value=0}]}}}
     Scale Curve=0
    )
    Index 1 Param 1=1
    Module.Index0Float0Write=true
    Module.Index0Float1Write=true
    Module.Index0Float2Write=true
    Module.Index0Float3Write=true
    Module.Index1Float0Write=true
    Module.Index1Float1Write=true
    Module.Index1Float2Write=true
    Module.Index1Float3Write=true
    Write Parameter Index 1=::NewEnumerator1
   )
   &ScaleColor(
    Scale Alpha=&FloatFromCurve(
     FloatCurve=DataInterface() {propertyValues {Curve {keys [
         {arriveTangent=1.1036, interpMode=RCIM_Cubic, leaveTangent=1.1036, tangentMode=RCTM_User, time=0.6, value=0}
         {arriveTangentWeight=0.0689, interpMode=RCIM_Cubic, leaveTangentWeight=0.5057, tangentMode=RCTM_Break, time=1.5, value=0.9}
         <at 1.9809 1 RCIM_Cubic>
        ]}}}
     CurveIndex=$Emitter.Age
     Scale Curve=0.4
    )
    Scale RGB=<1 1 2>
    ScaleA=true
    ScaleRGB=true
   )
   &SolveForcesAndVelocity(Clamp Velocity=false, Limit Acceleration=false)
  ]
  rendererClasses [&NiagaraMeshRendererProperties]
  renderers [{rendererClass=&NiagaraMeshRendererProperties, rendererIndex=0}]
 }
 SpawnDust {
  emitterUpdateScript [
   &EmitterState(Life Cycle Mode=::NewEnumerator1, Loop Behavior=::NewEnumerator1, Loop Delay=1.9, Loop Duration=1, UseLoopDelay=false)
   &SpawnBurst_Instantaneous(Spawn Count=10, Use Spawn Probability=false, UseLoopCountLimit=false)
  ]
  particleSpawnScript [
   &InitializeParticle(
    Color Minimum=<rgba 0.2587 0.2587 0.2587 1>
    Color Mode=::NewEnumerator2
    Lifetime Max=0.4
    Lifetime Min=0.6
    Lifetime Mode=::NewEnumerator1
    Sprite Rotation Mode=::NewEnumerator2
    Sprite Size Mode=::NewEnumerator1
    Uniform Sprite Size Max=60
    Uniform Sprite Size Min=40
    UsePositionOffset=false
    Write Material Random=true
   )
   &ShapeLocation(MaskSpawnGroup=false, Override Seed=false, Sphere Radius=15, UseOwnerScale=true)
   &CameraOffset(Camera Offset Amount=&RandomRangeFloat(Minimum=0, Maximum=10))
   &AddVelocity(
    Module.Use Falloff Distance=false
    Override Seed=false
    Velocity Mode=::NewEnumerator1
    Velocity Speed=&RandomRangeFloat(Minimum=60, Maximum=30)
   )
  ]
  particleUpdateScript [
   &ParticleState()
   &SpriteRotationRate(Rotation Rate=&RandomRangeFloat(Minimum=-150, Maximum=150))
   &ScaleSpriteSize(Uniform Curve Sprite Scale=DataInterface() {propertyValues {Curve {keys [
        {arriveTangent=1.6237, interpMode=RCIM_Cubic, leaveTangent=1.6237, tangentMode=RCTM_User, time=0, value=1}
        {arriveTangent=0.637, tangentMode=RCTM_User, time=1, value=1.7}
       ]}}})
   &Drag(
    Drag=&FloatFromCurve(FloatCurve=DataInterface() {propertyValues {Curve {keys [{interpMode=RCIM_Cubic, tangentMode=RCTM_User, time=0, value=0}, {interpMode=RCIM_Cubic, tangentMode=RCTM_User, time=0.4943, value=1}]}}})
    Use Linear Drag=true
    Use Rotational Drag=true
   )
   &SolveForcesAndVelocity(Clamp Velocity=false, Limit Acceleration=false)
   &ScaleColor(
    Scale Alpha=&FloatFromCurve(
     FloatCurve=DataInterface() {propertyValues {Curve {keys [
         {interpMode=RCIM_Cubic, tangentMode=RCTM_User, time=0, value=0.8902}
         {arriveTangent=-2.0189, interpMode=RCIM_Cubic, leaveTangent=-2.1015, tangentMode=RCTM_Break, time=0.2, value=1.8153}
         {arriveTangent=-1.1521, interpMode=RCIM_Cubic, tangentMode=RCTM_Break, time=1, value=0}
        ]}}}
     Scale Curve=&FloatFromCurve(
      FloatCurve=DataInterface() {propertyValues {Curve {keys [{interpMode=RCIM_Cubic, tangentMode=RCTM_User, time=4.5123, value=1}, {interpMode=RCIM_Cubic, tangentMode=RCTM_User, time=4.95, value=0}]}}}
      Scale Curve=&FNTimeOfDayScalar(Night Scale=0.6, Morning Scale=0.4, Day Scale=0.3, Evening Scale=0.4)
     )
    )
    Scale RGB=&MakeVectorFromLinearColorRGB(Color=&FNTimeOfDayColor(
      Night Color=&ColorFromCurve(ColorCurve=DataInterface() {propertyValues {
         AlphaCurve {keys [<at 0 1 RCIM_Cubic>]}
         BlueCurve {keys [<at 0 0.8096 RCIM_Cubic>,<at 1 0.3698 RCIM_Cubic>]}
         GreenCurve {keys [<at 0 0.4742 RCIM_Cubic>,<at 1 0.2166 RCIM_Cubic>]}
         RedCurve {keys [<at 0 0.4483 RCIM_Cubic>,<at 1 0.2048 RCIM_Cubic>]}
        }})
      Morning Color=&ColorFromCurve(ColorCurve=DataInterface() {propertyValues {
         AlphaCurve {keys [<at 0 1 RCIM_Cubic>]}
         BlueCurve {keys [<at 0 0.256 RCIM_Cubic>,<at 1 0.145 RCIM_Cubic>]}
         GreenCurve {keys [<at 0 0.3327 RCIM_Cubic>,<at 1 0.1631 RCIM_Cubic>]}
         RedCurve {keys [<at 0 0.8096 RCIM_Cubic>,<at 1 0.4161 RCIM_Cubic>]}
        }})
      Day Color=&ColorFromCurve(ColorCurve=DataInterface() {propertyValues {
         AlphaCurve {keys [<at 0 1 RCIM_Cubic>]}
         BlueCurve {keys [<at 0 0.0659 RCIM_Cubic>,<at 1 0.0754 RCIM_Cubic>]}
         GreenCurve {keys [<at 0 0.4818 RCIM_Cubic>,<at 1 0.3164 RCIM_Cubic>]}
         RedCurve {keys [<at 0 0.8096 RCIM_Cubic>,<at 1 0.8096 RCIM_Cubic>]}
        }})
      Evening Color=&ColorFromCurve(ColorCurve=DataInterface() {propertyValues {
         AlphaCurve {keys [<at 0 1 RCIM_Cubic>]}
         BlueCurve {keys [<at 0 0.3766 RCIM_Cubic>,<at 1 0.2029 RCIM_Cubic>]}
         GreenCurve {keys [<at 0 0.4646 RCIM_Cubic>,<at 1 0.2332 RCIM_Cubic>]}
         RedCurve {keys [<at 0 0.4693 RCIM_Cubic>,<at 1 0.2195 RCIM_Cubic>]}
        }})
     ))
    ScaleA=true
    ScaleRGB=true
   )
  ]
  rendererClasses [&NiagaraSpriteRendererProperties]
  renderers [{rendererClass=&NiagaraSpriteRendererProperties}]
 }
}
summary {
 systemName=NS_Emote_GnocchiTea_Glow
 emitters [
  {emitterName=GemGlows, rendererClasses [&NiagaraSpriteRendererProperties]}
  {emitterName=GemLights, rendererClasses [&NiagaraMeshRendererProperties]}
  {emitterName=LightSmoke, rendererClasses [&NiagaraSpriteRendererProperties]}
  {emitterName=Main_Ring, rendererClasses [&NiagaraMeshRendererProperties]}
  {emitterName=SpawnDust, rendererClasses [&NiagaraSpriteRendererProperties]}
  {emitterName=MotionBlur, rendererClasses [&NiagaraMeshRendererProperties]}
 ]
 userVariables [
  User.bIsFrontEnd {type=NiagaraBool}
  User.bIsFrontEndPreview {type=NiagaraBool}
  User.ParaGlideLeanAlpha {type=NiagaraFloat, defaultValue=0, description=""}
  User.RibbonIDOffset {type=NiagaraInt32}
  User.SkeletalMesh {
   type=NiagaraDataInterfaceSkeletalMesh
   defaultValue {
    struct=NiagaraExt_VariableValue_DataInterface
    value {dataInterface=&self:NiagaraDataInterfaceSkeletalMesh_1}
   }
  }
 ]
}
