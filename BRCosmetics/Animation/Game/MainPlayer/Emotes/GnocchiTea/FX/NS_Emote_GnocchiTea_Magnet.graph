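# Niagara system for the magnet effect of the GnocchiTea player emote, made of a warm glow and a trail.
# MainGlow burst-spawns a camera-offset sprite with curve-driven, flickering alpha.
# Trail continuously spawns short-lived ribbon particles with fixed velocity and facing, fading over emitter age with time-of-day-scaled alpha.
# The trail's skeletal mesh sampling of a gadget puzzle bone is disabled, as is its RGB scaling.
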NiagaraSystem /BRCosmetics/Animation/Game/MainPlayer/Emotes/GnocchiTea/FX/NS_Emote_GnocchiTea_Magnet
dependencies {
 usedDataInterfaces [&NiagaraDataInterfaceCurve, &NiagaraDataInterfaceSkeletalMesh, &NiagaraDataInterfaceColorCurve]
 usedDynamicInputs [&ColorFromCurve]
 usedModules [
  &SystemState
  &self:SystemSpawnScript.SystemScriptSource.SystemScriptGraph.NiagaraNodeAssignment_0.SetVariables_#1
 ]
}
emitters {
 MainGlow {
  emitterUpdateScript [
   &EmitterState(Life Cycle Mode=::NewEnumerator1, Loop Behavior=::NewEnumerator1, Loop Duration=2, UseLoopDelay=false)
   &SpawnBurst_Instantaneous(Use Spawn Probability=false, UseLoopCountLimit=false)
  ]
  particleSpawnScript [
   &InitializeParticle(
    Color=<rgba 1 0.5394 0.1682 0.5729>
    Lifetime=1.3
    Sprite Size Mode=::NewEnumerator3
    Uniform Sprite Size=120
    UsePositionOffset=false
    Write Material Random=true
   )
   &CameraOffset(Camera Offset Amount=10)
  ]
  particleUpdateScript [
   &ParticleState()
   &ScaleColor(
    Scale Alpha=&FloatFromCurve(
     FloatCurve=DataInterface() {propertyValues {Curve {keys [<at -0.0002 0>,<at 0.0983 0.3527>,<at 0.2007 0.1456>,<at 0.2797 0.4333>,<at 0.4132 0.3182>,<at 0.5389 0.6404>,<at 0.7021 0.5023>,<at 0.888 0.8241>,<at 1 0.4>]}}}
     Scale Curve=0.2
    )
    ScaleA=true
    ScaleRGB=true
   )
  ]
  rendererClasses [&NiagaraSpriteRendererProperties]
  renderers [{rendererClass=&NiagaraSpriteRendererProperties}]
 }
 Trail {
  emitterUpdateScript [
   &EmitterState(Inactive Response=::NewEnumerator1, Life Cycle Mode=::NewEnumerator1, Loop Behavior=::NewEnumerator1, Loop Duration=2, UseLoopDelay=false)
   &SpawnRate(SpawnRate=120, Use Spawn Probability=false)
  ]
  particleSpawnScript [
   &InitializeParticle(
    Color=<rgba 1 0.6827 0.3025 1>
    Lifetime=0.16
    Position Offset=<-7 0 0>
    Ribbon Width=80
    Ribbon Width Mode=::NewEnumerator1
    UsePositionOffset=true
    Write Material Random=true
   )
   &DynamicMaterialParameters(
    Index 0 Param 1=&RandomRangeFloat(Minimum=1, Maximum=1)
    Module.Index0Param0Write=true
    Module.Index0Param1Write=true
    Module.Index0Param2Write=true
    Module.Index0Param3Write=true
   )
   &SkeletalMeshLocation(
    Bone Sampling Mode=::NewEnumerator4
    MaskSpawnGroup=false
    Position Sampling=::NewEnumerator2
    Skeletal Mesh=DataInterface() {propertyValues {
      DefaultMesh=/BRCosmetics/Animation/Props/Gadgets_GnocchiTea/Meshes/GnocchiTea_Gadget
      ExcludeBoneName=root
      OptionalComponentClass=null
      PreviewMesh=&GnocchiTea_Gadget
      SkinningMode=SkinOnTheFly
      SoftSourceActor=null
      SourceMode=Default
      UvSetIndex=0
      WholeMeshLOD=-1
      bExcludeBone=false
      bReadDeformedGeometry=true
      bRequireCurrentFrameData=true
      ComponentTags []
      FilteredBones [JNT_InnerPuzzle3]
      FilteredSockets []
      MeshUserParameter {parameter {name=None, typeDefHandle {registeredTypeIndex=1}, varData []}}
      SamplingRegions []
     }}
   ) {enabled=false}
   &AddVelocity(Velocity=<-300 0 200>)
   /SetVariables(Particles.RibbonFacing=<0 1 0>) {isSetParameters=true}
  ]
  particleUpdateScript [
   &ParticleState()
   &ScaleColor(
    Scale Alpha=&FloatFromCurve(
     FloatCurve=DataInterface() {propertyValues {Curve {keys [
         {interpMode=RCIM_Cubic, tangentMode=RCTM_User, time=0, value=0}
         {arriveTangentWeight=0.462, interpMode=RCIM_Cubic, leaveTangentWeight=0.1836, tangentMode=RCTM_User, time=0.7536, value=1}
         {arriveTangentWeight=0.2404, interpMode=RCIM_Cubic, leaveTangentWeight=0.3606, tangentMode=RCTM_User, time=1.0531, value=1}
         {tangentMode=RCTM_User, time=1.3717, value=0}
        ]}}}
     CurveIndex=$Emitter.Age
     Scale Curve=&FNTimeOfDayScalar(Night Scale=0.3, Morning Scale=0.6, Day Scale=0.7, Evening Scale=0.6)
    )
    Scale RGB=&VectorFromFloat(Value=&FloatFromCurve(FloatCurve=DataInterface() {propertyValues {Curve {keys [
          <at 0 1 RCIM_Cubic>
          {arriveTangent=-2.3333, interpMode=RCIM_Cubic, leaveTangent=-2.3333, time=0.05, value=1}
          {arriveTangent=-1.0526, interpMode=RCIM_Cubic, leaveTangent=-1.0526, time=0.3, value=0.3}
          <at 1 0 RCIM_Cubic>
         ]}}}))
    ScaleA=true
    ScaleRGB=false
   )
   &SolveForcesAndVelocity(Clamp Velocity=false, Limit Acceleration=false)
  ]
  rendererClasses [&NiagaraRibbonRendererProperties]
  renderers [{rendererClass=&NiagaraRibbonRendererProperties}]
 }
}
summary {
 systemName=NS_Emote_GnocchiTea_Magnet
 emitters [
  {emitterName=MainGlow, rendererClasses [&NiagaraSpriteRendererProperties]}
  {emitterName=Trail, rendererClasses [&NiagaraRibbonRendererProperties]}
 ]
 userVariables []
}
